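; Quest ability for BattlePassS24 that credits the owning player when a target player's shields are cracked.
; On a gameplay event it calls the parent, prints the target's name as debug output, and checks the event tag against the Other Container tags.
; If the target is a valid Athena player pawn with zero shield and overshield, it sends a custom stat event to the quest manager with additional source tags, then ends.
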(event Ability|EventActivateAbilityFromEvent
  (params (EventData :type Gameplay%20Event%20Data%20Structure%20%28by%20ref%29))
  (bind (_eventtag _instigator _target) (Utilities|Struct|BreakGameplayEventData EventData))
  (bind _asplayer_pawn_athena (Utilities|Casting|CastToPlayerPawn_Athena _target))
  (|CallParentFunction EventData)
  (Development|PrintString (Utilities|GetDisplayName _target))
  (if (GameplayTags|MatchesAnyTags _eventtag (Variables|Default|GetOther%20Container))
    (Utilities|IsValid _asplayer_pawn_athena
      (:Is%20Valid
        (if (and (== (FortPawn|GetShield _asplayer_pawn_athena) 0) (== (FortPawn|GetOvershield _asplayer_pawn_athena) 0))
          (Quest|SendComplexCustomStatEvent (Variables|GAVars|Data|GetOwningPlayerQuestManager) 0 (Variables|Default|GetAdditional%20Source%20Tags) "(GameplayTags=)" 1)
          (Ability|EndAbility)
          (else
            _)))
      (:Is%20Not%20Valid
        _)))
  (Ability|EndAbility))
